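; HUD progress-bar view model that displays the pawn's tactical sprint energy while Masamune is equipped, using a Masamune sprint material.
; On initialize it caches the TacticalSprintEnergyComponent and starts paired gameplay tag query waits that toggle the bar on and off.
; Toggling binds or unbinds the energy-changed event feeding progress, resetting it to zero when disabled.
; An immediate-visibility check reports true when energy is below full.
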(event EventInitialize
  (params (Pawn :type Pawn%20Object%20Reference))
  (|CallParentFunction Pawn)
  (Utilities|IsValid Pawn
    (:Is%20Valid
      (Variables|Default|SetTacticalSprintComponent (Actor|GetComponentbyClass Pawn "/TacticalSprintGame/Gameplay/TacticalSprintEnergyComponent.TacticalSprintEnergyComponent_C"))
      ;[1 0.8368 0.06771 0.5] Listening to when Masamune is equipped
      (Ability|Tasks|WaitGameplayTagQueryOnActor Pawn (Variables|Default|GetMasamuneEnergyTagQuery)
        (:then
          (Ability|Tasks|WaitGameplayTagQueryOnActor Pawn (Variables|Default|GetMasamuneEnergyTagQuery) "WhenFalse"
            (:then
              (CallFunction|SetProgressBarOverrideMaterial "/MasamuneGameplay/UI/Materials/MI_UI_Masamune_Sprint_Rounded.MI_UI_Masamune_Sprint_Rounded"))
            (:Triggered
              (CallFunction|ToggleMasamuneEnergyBar false))))
        (:Triggered
          (CallFunction|ToggleMasamuneEnergyBar true))))))

(event EventUninitialize
  (params (Pawn :type Pawn%20Object%20Reference))
  (bind _tacticalsprintcomponent (Variables|Default|GetTacticalSprintComponent))
  (Utilities|IsValid _tacticalsprintcomponent
    (:Is%20Valid
      (Events|UnbindEventfromOnCurrentEnergyChanged _tacticalsprintcomponent (EventDispatchers|CreateEvent :!event OnCurrentEnergyChanged)))))

(fn CreateEnergyChangedBinding
  (bind _tacticalsprintcomponent (Variables|Default|GetTacticalSprintComponent))
  (Utilities|IsValid _tacticalsprintcomponent
    (:Is%20Valid
      (CallFunction|OnCurrentEnergyChanged _tacticalsprintcomponent)
      (Events|BindEventtoOnCurrentEnergyChanged _tacticalsprintcomponent (EventDispatchers|CreateEvent :!event OnCurrentEnergyChanged)))))

(fn OnCurrentEnergyChanged
  (params (EnergyComponent :type Fort%20Component%20Energy%20Object%20Reference) (PreviousValue :type Float%20%28single-precision%29))
  (Variables|GenericProgressBarViewModel|SetProgressValue (Energy|GetCurrentEnergyPercentage EnergyComponent)))

(fn RemoveEnergyChangedBinding
  (bind _tacticalsprintcomponent (Variables|Default|GetTacticalSprintComponent))
  (Utilities|IsValid _tacticalsprintcomponent
    (:Is%20Valid
      (Events|UnbindEventfromOnCurrentEnergyChanged _tacticalsprintcomponent (EventDispatchers|CreateEvent :!event OnCurrentEnergyChanged))
      (Variables|GenericProgressBarViewModel|SetProgressValue 0.0))))

(fn ShouldBecomeVisibleImmediately
  (bind _tacticalsprintcomponent (Variables|Default|GetTacticalSprintComponent))
  (Utilities|IsValid _tacticalsprintcomponent
    (:Is%20Valid
      (return 0.0 (< (Energy|GetCurrentEnergyPercentage _tacticalsprintcomponent) 1.0)))
    (:Is%20Not%20Valid
      (return 0.0 false))))

(fn ToggleMasamuneEnergyBar
  (params (bIsActive :type Boolean))
  (if bIsActive
    (CallFunction|CreateEnergyChangedBinding)
    (Variables|GenericProgressBarViewModel|SetbEnabled true)
    (else
      (CallFunction|RemoveEnergyChangedBinding)
      (Variables|GenericProgressBarViewModel|SetbEnabled false))))
